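# Niagara idle effect for the SpeedDrive guitar cosmetic, continuously spawning short-lived wind whirl meshes from a single Windwhirl emitter.
# Particles split into five partitions that pick local position offsets from an array and vary mesh scale.
# Each mesh spins at a random negative rate while growing and fading over its life.
# Color and alpha follow Fortnite time-of-day curves, with User.bIsFrontEndPreview swapping in a fixed tint.
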NiagaraSystem /SparksCosmetics/Instrument/Guitar/SpeedDrive/Effects/NS_Guitar_SpeedDrive_IdleFX
dependencies {
 usedDataInterfaces [
  &NiagaraDataInterfaceArrayFloat3
  &NiagaraDataInterfaceColorCurve
  &NiagaraDataInterfaceCurve
  &NiagaraDataInterfaceSkeletalMesh
 ]
 usedDynamicInputs [&MakeFloatFromVector]
 usedModules [
  &self:SystemSpawnScript.SystemScriptSource.SystemScriptGraph.NiagaraNodeAssignment_6.SetVariables_#1
  &SystemState
  &self:SystemSpawnScript.SystemScriptSource.SystemScriptGraph.NiagaraNodeAssignment_0.SetVariables_#2
  &self:SystemSpawnScript.SystemScriptSource.SystemScriptGraph.NiagaraNodeAssignment_1.SetVariables_#3
 ]
}
emitters {Windwhirl {
  emitterUpdateScript [
   /SetVariables(Emitter.Spline Width:Vector4f=/Niagara/DynamicInputs/Multiply/Multiply_Vector4_ByFloat(Vector4:Vector4f=<2.75 2.75 0 0>, Float:NiagaraFloat=$System.MeshScale)) {isSetParameters=true}
   &EmitterState()
   &SpawnRate(SpawnRate=40, Use Spawn Probability=false)
  ]
  particleSpawnScript [
   &PartitionParticles(Number Of Partitions=5)
   /SetVariables() {isSetParameters=true}
   &InitializeParticle(
    Color=<rgba 0.4905 0.4905 0.6562 1>
    Lifetime Max=0.55
    Lifetime Min=0.35
    Lifetime Mode=::NewEnumerator1
    Mesh Scale Mode=::NewEnumerator3
    Mesh Uniform Scale=&MaskFloat(
     Mask Source=1
     Mask Parameter=$Particles.PartitionParticles.Partition
     Mask=0
     Masked Float=&RandomRangeFloat(Minimum=0.14, Maximum=0.19)
     Passthrough Float=&MaskFloat(
      Mask Source=1
      Mask Parameter=$Particles.PartitionParticles.Partition
      Mask=1
      Masked Float=&RandomRangeFloat(Minimum=0.14, Maximum=0.185)
      Passthrough Float=&RandomRangeFloat(Minimum=0.2, Maximum=0.2)
     )
    )
    Position Offset=&SelectVectorFromArray(
     Direct Array Index=$Particles.PartitionParticles.Partition
     Vector Selection Array=DataInterface() {propertyValues {FloatData [<0 -0.9 -9>, <0 -6 -9>, <0 -2.5 -10>, <0 -4 -9>, <0 -5.25 -9>]}}
    )
    Position Offset Coordinate Space=::Local
    UsePositionOffset=true
    Write Material Random=true
   )
   &ShapeLocation(
    Cylinder Height=10
    Cylinder Radius=5
    MaskSpawnGroup=false
    Override Seed=false
    Rotation Angle=90
    Rotation Axis=<0 1 0>
    Rotation Mode=::NewEnumerator1
    Shape Primitive=::NewEnumerator1
    Surface Only=false
    UseOwnerScale=true
   ) {enabled=false}
   &InitialMeshOrientation(
    Mesh Axis To Orient=<0 1 0>
    Mesh Orientation Mode=::NewEnumerator2
    Orientation Vector=<0 0 1>
    Rotation=&RandomRangeVector(Minimum=<0 0 0>, Maximum=<0 0 1>)
    Use Rotation Vector=true
   )
  ]
  particleUpdateScript [
   &ParticleState()
   &ScaleColor(
    Scale Alpha=&FloatFromCurve(
     FloatCurve=DataInterface() {propertyValues {Curve {keys [
         {arriveTangent=9.0102, interpMode=RCIM_Cubic, leaveTangent=9.0102, tangentMode=RCTM_User, time=0, value=0}
         {arriveTangent=2.4457, interpMode=RCIM_Cubic, leaveTangent=2.4457, tangentMode=RCTM_User, time=0.1, value=0.75}
         {arriveTangentWeight=0.1333, interpMode=RCIM_Cubic, leaveTangentWeight=0.1667, tangentMode=RCTM_User, time=0.5, value=1}
         {arriveTangent=-3.2028, interpMode=RCIM_Cubic, leaveTangent=-3.2028, tangentMode=RCTM_User, time=1, value=0}
        ]}}}
     Scale Curve=&FloatFromCurve(
      FloatCurve=DataInterface() {propertyValues {Curve {keys [
          <at 0 0.1362 RCIM_Cubic>
          {arriveTangentWeight=0.1348, interpMode=RCIM_Cubic, leaveTangentWeight=0.0674, tangentMode=RCTM_User, time=0.4, value=0.1562}
          {arriveTangentWeight=0.0674, interpMode=RCIM_Cubic, leaveTangentWeight=0.1348, tangentMode=RCTM_User, time=0.6, value=0.1562}
          <at 1 0.1362 RCIM_Cubic>
         ]}}}
      CurveIndex=Linked() {linkedVariable {name=NPC.FortniteNPC.TimeOfDay_0-1, type=NiagaraFloat}}
     )
    )
    Scale RGB=&MakeVectorFromLinearColorRGB(Color=&MakeCustomLinearColorFromBool(
      Bool=$User.bIsFrontEndPreview
      True Color=<rgba 0.7516 0.8273 1 1>
      False Color=&ColorFromCurve(
       ColorCurve=DataInterface() {propertyValues {
         AlphaCurve {keys [<at 0 1 RCIM_Cubic>]}
         BlueCurve {keys [<at 0.07 1 RCIM_Cubic>,<at 0.125 0.6>,<at 0.25 0.7>,<at 0.5 1>,<at 0.625 0.9>,<at 0.75 0.9891>,<at 1 1 RCIM_Cubic>]}
         GreenCurve {keys [<at 0.07 0.4436 RCIM_Cubic>,<at 0.125 0.7169>,<at 0.25 0.8561>,<at 0.5 1>,<at 0.625 0.9522>,<at 0.75 0.8087>,<at 1 0.4436 RCIM_Cubic>]}
         RedCurve {keys [<at 0.07 0.04 RCIM_Cubic>,<at 0.125 1>,<at 0.25 1>,<at 0.5 1>,<at 0.625 1>,<at 0.75 1>,<at 1 0.04 RCIM_Cubic>]}
        }}
       CurveIndex=Linked() {linkedVariable {name=NPC.FortniteNPC.TimeOfDay_0-1, type=NiagaraFloat}}
      )
     ))
    ScaleA=true
    ScaleRGB=true
   )
   &UpdateMeshOrientation(
    Rotation Coordinate Space=::NewEnumerator3
    Rotation Rate=&RandomRangeFloat(Minimum=-7, Maximum=-15)
    Rotation Vector=&RandomRangeVector(Minimum=<0 0 0>, Maximum=<0 0 1>)
   )
   &ScaleMeshSize(Scale Factor=&Multiply_VectorByFloat(
     Vector=<1 1 1>
     Float=&FloatFromCurve(FloatCurve=DataInterface() {propertyValues {Curve {keys [{arriveTangent=3.0713, interpMode=RCIM_Cubic, leaveTangent=3.0713, tangentMode=RCTM_User, time=0, value=0}, <at 0.6 1 RCIM_Cubic>]}}})
    ))
  ]
  rendererClasses [&NiagaraMeshRendererProperties]
  renderers [{rendererClass=&NiagaraMeshRendererProperties, rendererIndex=0}]
 }}
summary {
 systemName=NS_Guitar_SpeedDrive_IdleFX
 emitters [{emitterName=Windwhirl, rendererClasses [&NiagaraMeshRendererProperties]}]
 userVariables [User.bIsFrontEndPreview {type=NiagaraBool}]
}
