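# Niagara swing trail for the SpeedDrive guitar pickaxe, built from socket-driven anim-trail ribbons and sprite sparks.
# User.IsVariant switches between the Default_Trail and Variant_Trail ribbons by zeroing one's trail count, and selects red-orange or cyan-blue spark colors.
# Sparks inherit socket motion, then scatter under drag, gravity and curl noise.
# A narrow, dark TireTrack ribbon gates spawn alpha by emitter age, and all emitters tint by time-of-day color.
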NiagaraSystem /SparksCosmetics/Instrument/Guitar/SpeedDrive/Effects/NS_Pickaxe_Guitar_SpeedDrive_Trail
dependencies {
 usedDataInterfaces [
  &NiagaraDataInterfaceCurve
  &NiagaraDataInterfaceColorCurve
  &NiagaraDataInterfaceVector2DCurve
  &NiagaraDataInterfaceSocketReader
 ]
 usedDynamicInputs []
 usedModules [
  &self:SystemSpawnScript.SystemScriptSource.SystemScriptGraph.NiagaraNodeAssignment_0.SetVariables_#1
  &self:SystemSpawnScript.SystemScriptSource.SystemScriptGraph.NiagaraNodeAssignment_1.SetVariables_#2
  &self:SystemSpawnScript.SystemScriptSource.SystemScriptGraph.NiagaraNodeAssignment_2.SetVariables_#3
  &SystemState
 ]
}
emitters {
 Default_Trail {
  emitterUpdateScript [
   &EmitterState(Life Cycle Mode=::NewEnumerator1, Loop Behavior=::NewEnumerator1, Loop Duration=0.22, UseLoopDelay=true)
   &FN_SpawnAnimTrail_v2(Number of Anim Trails:NiagaraInt32=&MakeCustomIntFromBool(Bool=&SetBoolByFloatComparison(A=$User.IsVariant, B=0), True INT=0, False INT=1))
  ]
  particleSpawnScript [
   /Game/Effects/Niagara/Modules/AnimTrail/FN_InitializeAnimTrail/FN_InitializeAnimTrailBoneIndecies(First Bone or Socket Index=1)
   /Game/Effects/Niagara/Modules/AnimTrail/FN_InitializeAnimTrail/FN_InitializeAnimTrail(AnimTrailWidth=1, Lifetime=0.2)
   &SystemLocation(MaskSpawnGroup=false)
  ]
  particleUpdateScript [
   &ScaleRibbonWidth(Ribbon Width Scale=&FloatFromCurve(FloatCurve=DataInterface() {propertyValues {Curve {keys [
         {arriveTangent=-1.2579, interpMode=RCIM_Cubic, time=0, value=1}
         {arriveTangent=-0.9664, interpMode=RCIM_Cubic, leaveTangent=-0.9664, tangentMode=RCTM_User, time=1, value=0}
        ]}}}))
   &ScaleColor(
    Scale Alpha=&Multiply_Float(
     A=&FloatFromCurve(
      FloatCurve=DataInterface() {propertyValues {Curve {keys [
          {interpMode=RCIM_Cubic, leaveTangent=-1.2334, tangentMode=RCTM_Break, time=0, value=1}
          <at 0.6689 0.1>
          {arriveTangent=-0.2139, interpMode=RCIM_Cubic, tangentMode=RCTM_Break, time=1, value=0}
         ]}}}
      Scale Curve=&FloatFromCurve(
       FloatCurve=DataInterface() {propertyValues {Curve {keys [<at 0 1>,<at 0.25 1>,<at 0.5 1>,<at 0.75 1>,<at 1 1>]}}}
       CurveIndex=Linked() {linkedVariable {name=NPC.FortniteNPC.TimeOfDay_0-1, type=NiagaraFloat}}
      )
     )
     B=&FN_FrontEndFloatModifier(Gameplay=0.5, Front End=0.35)
    )
    Scale RGB=&MakeVectorFromLinearColorRGB(Color=$System.TOD Color)
    ScaleA=true
    ScaleRGB=true
   ) {module=ScaleColor002}
   &FN_UpdateAnimTrail_v2()
  ]
  rendererClasses [&NiagaraRibbonRendererProperties]
  renderers [{rendererClass=&NiagaraRibbonRendererProperties}]
 }
 Swing_Sparks {
  emitterUpdateScript [
   &EmitterState(Life Cycle Mode=::NewEnumerator1, Loop Behavior=::NewEnumerator1, Loop Delay=0.01, Loop Duration=1, UseLoopDelay=true)
   &SpawnRate(
    Spawn Probability=0.9
    SpawnRate=&RandomRangeFloat(Minimum=20, Maximum=125, Evaluation Type=::NewEnumerator1)
    Use Spawn Probability=false
   )
   &SpawnPerUnit(
    Max Movement Threshold=2000
    Module.Use Max Movement Threshold=true
    Module.Use Movement Tolerance=false
    Spawn Spacing=50
    Use Spawn Probability=true
   ) {enabled=false}
  ]
  particleSpawnScript [
   &InitializeParticle(
    Color=&MakeCustomLinearColorFromBool(
     Bool=&SetBoolByFloatComparison(A=$User.IsVariant, B=0)
     True Color=&RandomRangeLinearColor(Minimum=<rgba 0 1.9455 2 1>, Maximum=<rgba 0 3.3215 5 1>)
     False Color=&RandomRangeLinearColor(Minimum=<rgba 2 0 0.024 1>, Maximum=<rgba 5 2.5421 0 1>)
    )
    Lifetime Max=0.3
    Lifetime Min=0.15
    Lifetime Mode=::NewEnumerator1
    Sprite Rotation Mode=::NewEnumerator2
    Sprite Size=&Multiply_Vector2D_ByFloat(Float=&RandomRangeFloat(Minimum=0.75, Maximum=4), Vector2D=<1 1.5>)
    Sprite Size Mode=::NewEnumerator4
    UsePositionOffset=false
    Write Material Random=true
   )
   &SocketLocation(
    Interpolate Socket Sampling=true
    SamplePosOffset=true
    Sampled Position Offset=<0 0 -7>
    Socket Index=1
    Socket Reader=$System.Socket Reader
    Socket Sampling Mode=::NewEnumerator3
    VelocitySpeedLimit=false
   )
   &ShapeLocation(MaskSpawnGroup=false, Override Seed=false, Sphere Radius=5, UseOwnerScale=true)
   &InheritVelocity(Inherited Velocity Amount Scale=<0.05 0.05 0.05>, Speed Limit=true, UseSpeedThreshold=false)
   &AddVelocity(Rotation Coordinate Space=::World, Velocity=$Output.SocketLocation.SampledVelocity, Velocity Speed Scale=0.07)
   &ApplyInitialForces()
   &DynamicMaterialParameters(
    Index 0 Param 1=3
    Index 0 Param 2=20
    Index 0 Param 3=1
    Module.Index0Float0Write=true
    Module.Index0Float1Write=true
    Module.Index0Float2Write=true
    Module.Index0Float3Write=true
   ) {module=DynamicMaterialParameters001}
  ]
  particleUpdateScript [
   &ParticleState()
   &ScaleSpriteSize(Uniform Curve Sprite Scale=DataInterface() {propertyValues {Curve {
       defaultValue=0
       keys [
        {interpMode=RCIM_Cubic, tangentMode=RCTM_User, time=0.1, value=1}
        {arriveTangent=-3.2192, interpMode=RCIM_Cubic, leaveTangent=-3.2192, tangentMode=RCTM_User, time=1, value=0}
       ]
      }}})
   &Drag(Drag=&RandomRangeFloat(Minimum=7, Maximum=12), Use Linear Drag=true, Use Rotational Drag=false)
   &GravityForce(Gravity=&Multiply_VectorByFloat(
     Vector=&RandomRangeVector(Minimum=<0 0 -100>, Maximum=<0 0 -500>)
     Float=&FloatFromCurve(FloatCurve=DataInterface() {propertyValues {Curve {
         defaultValue=1
         keys [
          {arriveTangent=1.7336, interpMode=RCIM_Cubic, leaveTangent=1.7336, tangentMode=RCTM_User, time=0.1, value=0}
          {interpMode=RCIM_Cubic, tangentMode=RCTM_User, time=1, value=1}
         ]
        }}})
    ))
   &CurlNoiseForce(Bias Noise=false, Noise Frequency=10, Noise Strength=400, Pan Noise=false)
   &SolveForcesAndVelocity(Clamp Velocity=false, Limit Acceleration=false)
   &ScaleSpriteSizeBySpeed(Max Scale Factor=<0.9 1.2>, Min Scale Factor=<0.5 1>)
   &ScaleColor(
    Scale Alpha=&FloatFromCurve(FloatCurve=DataInterface() {propertyValues {Curve {
        defaultValue=1
        keys [{interpMode=RCIM_Cubic, tangentMode=RCTM_User, time=0.2, value=1}, {interpMode=RCIM_Cubic, tangentMode=RCTM_User, time=1, value=0}]
       }}})
    Scale RGB=&Multiply_Vector(
     A=&MakeCustomVectorFromBool(
      Bool=&SetBoolByFloatComparison(A=$User.IsVariant, B=0)
      True Vector=&MakeVectorFromLinearColorRGB(Color=&ColorFromCurve(ColorCurve=DataInterface() {propertyValues {
          AlphaCurve {keys [<at 0 1 RCIM_Cubic>]}
          BlueCurve {keys [<at 0.0404 0.434 RCIM_Cubic>,<at 0.3668 1>,<at 0.9206 0.5885 RCIM_Cubic>]}
          GreenCurve {keys [<at 0.0404 2 RCIM_Cubic>,<at 0.3668 0.9952>,<at 0.9206 0 RCIM_Cubic>]}
          RedCurve {keys [<at 0.0404 0.2099 RCIM_Cubic>,<at 0.3668 0.135>,<at 0.9206 0.0083 RCIM_Cubic>]}
         }}))
      False Vector=&MakeVectorFromLinearColorRGB(Color=&ColorFromCurve(ColorCurve=DataInterface() {propertyValues {
          AlphaCurve {keys [<at 0 1 RCIM_Cubic>]}
          BlueCurve {keys [<at 0.0402 1.3065 RCIM_Cubic>,<at 0.3526 0>,<at 0.9126 0.0279 RCIM_Cubic>]}
          GreenCurve {keys [<at 0.0402 1.8536 RCIM_Cubic>,<at 0.3526 0.3325>,<at 0.9126 0 RCIM_Cubic>]}
          RedCurve {keys [<at 0.0402 2 RCIM_Cubic>,<at 0.3526 1>,<at 0.9126 0.5885 RCIM_Cubic>]}
         }}))
     )
     B=&MakeVectorFromLinearColorRGB(Color=$System.TOD Color)
    )
    ScaleA=true
    ScaleRGB=true
   )
  ]
  rendererClasses [&NiagaraSpriteRendererProperties]
  renderers [{rendererClass=&NiagaraSpriteRendererProperties}]
 }
 TireTrack {
  emitterSpawnScript [
   &FN_RibbonIDOffsetUtility()
   /SetVariables(Emitter.RibbonLife=0.22) {isSetParameters=true}
  ]
  emitterUpdateScript [
   &EmitterState()
   &FN_SpawnAnimTrail(AnimTrailWidthMode=::FromFirst)
  ]
  particleSpawnScript [
   /Game/Effects/Niagara/Modules/AnimTrail/FN_InitializeAnimTrailBoneIndecies_v2/FN_InitializeAnimTrailBoneIndecies(First Bone or Socket Index=1, Second Bone or Socket Index=0)
   /Game/Effects/Niagara/Modules/AnimTrail/FN_InitializeAnimTrail/FN_InitializeAnimTrail_v5/FN_InitializeAnimTrail(AnimTrailWidth=0.25, Color=$System.TOD Color, Lifetime=$Emitter.RibbonLife)
   /SetVariables(Particles.Spawn Alpha:NiagaraFloat=&FloatFromCurve(
     FloatCurve=DataInterface() {propertyValues {Curve {
        defaultValue=1
        keys [
         {interpMode=RCIM_Cubic, leaveTangentWeight=0.0345, tangentMode=RCTM_User, time=0.06, value=0}
         {arriveTangentWeight=0.0909, interpMode=RCIM_Cubic, leaveTangentWeight=0.1946, tangentMode=RCTM_User, time=0.1016, value=1}
         {arriveTangentWeight=0.0176, interpMode=RCIM_Cubic, tangentMode=RCTM_User, time=0.1367, value=0}
        ]
       }}}
     CurveIndex=$Emitter.Age
    )) {isSetParameters=true}
  ]
  particleUpdateScript [
   &SolveForcesAndVelocity(Clamp Velocity=false, Limit Acceleration=false)
   &FN_UpdateAnimTrail(Lock Particles=true) {module=FN_UpdateAnimTrail001}
   &ScaleColor(Scale Mode=2) {enabled=false}
   &ScaleColor(
    Scale Alpha=&Multiply_Float(
     A=&FloatFromCurve(
      FloatCurve=DataInterface() {propertyValues {Curve {keys [
          {arriveTangent=4.7872, interpMode=RCIM_Cubic, leaveTangent=4.7872, tangentMode=RCTM_User, time=0, value=0}
          {arriveTangentWeight=0.0648, interpMode=RCIM_Cubic, leaveTangentWeight=0.3239, tangentMode=RCTM_User, time=0.1, value=1}
          {arriveTangentWeight=0.2491, interpMode=RCIM_Cubic, leaveTangentWeight=0.1993, tangentMode=RCTM_User, time=0.6, value=1}
          <at 1 0 RCIM_Cubic>
         ]}}}
      CurveIndex=&Divide_Float(A=$Emitter.Age, B=$Emitter.RibbonLife)
     )
     B=$Particles.Initial.Spawn Alpha
    )
    Scale RGB=&Multiply_Vector(
     A=&MakeVectorFromLinearColorRGB(Color=$System.TOD Color)
     B=&MakeVectorFromLinearColorRGB(Color=&ColorFromCurve(
       ColorCurve=DataInterface() {propertyValues {
         AlphaCurve {keys [<at 0 1 RCIM_Cubic>]}
         BlueCurve {keys [<at 0.2647 0.0104 RCIM_Cubic>,<at 1 0.0417 RCIM_Cubic>]}
         GreenCurve {keys [<at 0.2647 0.0069 RCIM_Cubic>,<at 1 0.0347 RCIM_Cubic>]}
         RedCurve {keys [<at 0.2647 0.0078 RCIM_Cubic>,<at 1 0.0373 RCIM_Cubic>]}
        }}
       CurveIndex=&Frac_Float(Float=&Multiply_Float(A=$Engine.System.Age, B=7))
      ))
    )
    ScaleA=true
    ScaleRGB=true
   ) {module=ScaleColor005}
   &ScaleColor(
    Scale Mode=::NewEnumerator1
    Scale RGBA=&MakeVector4_FromLinearColor(LinearColor=$System.TOD Color)
    ScaleRGBA=true
   ) {enabled=false, module=ScaleColor004}
   &ScaleColor(
    Scale Alpha=&FloatFromCurve(FloatCurve=DataInterface() {propertyValues {Curve {keys [<at 0 0.2286>,<at 0.094 0.9936>,<at 0.4 1>,<at 0.56 0.9>,<at 1 0>]}}})
    ScaleA=true
    ScaleRGB=true
   ) {enabled=false, module=ScaleColor003}
   /SetVariables(Particles.RibbonUVDistance:NiagaraFloat=&Multiply_FloatByInt(Float=1.4, Integer=&ReturnParticleIDAttributes())) {isSetParameters=true}
   /SetVariables(Particles.ScreenRibbonWidth:NiagaraFloat=&Multiply_Float(A=$Particles.RibbonWidth, B=1)) {isSetParameters=true}
  ]
  rendererClasses [&NiagaraRibbonRendererProperties]
  renderers [{rendererClass=&NiagaraRibbonRendererProperties}]
 }
 Variant_Trail {
  emitterUpdateScript [
   &EmitterState(Life Cycle Mode=::NewEnumerator1, Loop Behavior=::NewEnumerator1, Loop Duration=0.22, UseLoopDelay=true)
   &FN_SpawnAnimTrail_v2(Number of Anim Trails:NiagaraInt32=&MakeCustomIntFromBool(Bool=&SetBoolByFloatComparison(A=$User.IsVariant, B=0), True INT=1, False INT=0))
  ]
  particleSpawnScript [
   /Game/Effects/Niagara/Modules/AnimTrail/FN_InitializeAnimTrail/FN_InitializeAnimTrailBoneIndecies(First Bone or Socket Index=1)
   /Game/Effects/Niagara/Modules/AnimTrail/FN_InitializeAnimTrail/FN_InitializeAnimTrail(AnimTrailWidth=1, Lifetime=0.2)
   &SystemLocation(MaskSpawnGroup=false)
  ]
  particleUpdateScript [
   &ScaleRibbonWidth(Ribbon Width Scale=&FloatFromCurve(FloatCurve=DataInterface() {propertyValues {Curve {keys [
         {arriveTangent=-1.2579, interpMode=RCIM_Cubic, time=0, value=1}
         {arriveTangent=-0.9664, interpMode=RCIM_Cubic, leaveTangent=-0.9664, tangentMode=RCTM_User, time=1, value=0}
        ]}}}))
   &ScaleColor(
    Scale Alpha=&Multiply_Float(
     A=&FloatFromCurve(
      FloatCurve=DataInterface() {propertyValues {Curve {keys [
          {interpMode=RCIM_Cubic, leaveTangent=-1.2334, tangentMode=RCTM_Break, time=0, value=1}
          <at 0.6689 0.1>
          {arriveTangent=-0.2139, interpMode=RCIM_Cubic, tangentMode=RCTM_Break, time=1, value=0}
         ]}}}
      Scale Curve=&FloatFromCurve(
       FloatCurve=DataInterface() {propertyValues {Curve {keys [<at 0 1>,<at 0.25 1>,<at 0.5 1>,<at 0.75 1>,<at 1 1>]}}}
       CurveIndex=Linked() {linkedVariable {name=NPC.FortniteNPC.TimeOfDay_0-1, type=NiagaraFloat}}
      )
     )
     B=&FN_FrontEndFloatModifier(Gameplay=0.5, Front End=0.35)
    )
    Scale RGB=&MakeVectorFromLinearColorRGB(Color=$System.TOD Color)
    ScaleA=true
    ScaleRGB=true
   ) {module=ScaleColor002}
   &FN_UpdateAnimTrail_v2()
  ]
  rendererClasses [&NiagaraRibbonRendererProperties]
  renderers [{rendererClass=&NiagaraRibbonRendererProperties}]
 }
}
summary {
 systemName=NS_Pickaxe_Guitar_SpeedDrive_Trail
 emitters [
  {emitterName=Default_Trail, rendererClasses [&NiagaraRibbonRendererProperties]}
  {emitterName=TireTrack, rendererClasses [&NiagaraRibbonRendererProperties]}
  {emitterName=Swing_Sparks, rendererClasses [&NiagaraSpriteRendererProperties]}
  {emitterName=Variant_Trail, rendererClasses [&NiagaraRibbonRendererProperties]}
 ]
 userVariables [
  User.bIsFrontEnd {type=NiagaraBool}
  User.bIsFrontEndPreview {type=NiagaraBool}
  User.IsVariant {type=NiagaraFloat, defaultValue=0, description=""}
  User.RibbonIDOffset {type=NiagaraInt32}
  User.SkeletalMesh {
   type=NiagaraDataInterfaceSkeletalMesh
   defaultValue {
    struct=NiagaraExt_VariableValue_DataInterface
    value {dataInterface=&self:NiagaraDataInterfaceSkeletalMesh_0}
   }
  }
 ]
}
