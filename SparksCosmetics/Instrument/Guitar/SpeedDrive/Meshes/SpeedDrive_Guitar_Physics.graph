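# Physics asset for the SpeedDrive guitar, a Sparks instrument cosmetic, providing collision and a jointed tire part.
# It has two capsule bodies: a large one along the handle and a small offset one on the dyn_tire bone.
# A single constraint joins the tire to the handle with Swing2 free and Swing1 and twist locked, so the tire rotates about one axis like a hinge.
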PhysicsAsset /SparksCosmetics/Instrument/Guitar/SpeedDrive/Meshes/SpeedDrive_Guitar_Physics
bodies {
 dyn_tire [{shapeType=Capsule, length=2, radius=4, shapeName=dyn_tire_capsule, center=<-20 0 0>, rotation=<rot 0 0 90>}]
 handle [{shapeType=Capsule, length=50.5, radius=10.5, center=<0 0 31.9279>}]
}
constraints [{
  bone1Name=dyn_tire
  bone2Name=handle
  swing1LimitDegrees=45
  swing1Motion=Locked
  swing2LimitDegrees=180
  swing2Motion=Free
  twistLimitDegrees=180
  twistMotion=Locked
 }]
